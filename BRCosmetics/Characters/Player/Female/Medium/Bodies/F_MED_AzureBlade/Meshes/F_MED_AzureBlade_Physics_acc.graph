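# Physics asset for accessories on the F_MED_AzureBlade female player body, covering a set of dangling charms.
# A two-link ring chain hangs from a base bone constrained to spine_02, each link swinging on one alternating axis, then branches into two-bone A and three-bone B charm strands with tight swing and twist limits.
# Capsules on the pelvis, lower spine, right thigh and right skirt form the surrounding body shapes, joined by free angular constraints.
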PhysicsAsset /BRCosmetics/Characters/Player/Female/Medium/Bodies/F_MED_AzureBlade/Meshes/F_MED_AzureBlade_Physics_acc
bodies {
 base_charms_ring_1 [{shapeType=Capsule, length=0.5, radius=0.25, shapeName=dyn_charms_ring_2_capsule}]
 dyn_charms_A_1 [{shapeType=Capsule, length=2.4927, radius=0.5, shapeName=dyn_charms_A_1_capsule, center=<1.3623 0 0>, rotation=<rot 90 0 0>}]
 dyn_charms_A_2 [{shapeType=Capsule, length=6.4798, radius=1.7197, shapeName=dyn_charms_A_1_capsule, center=<4.5546 -0.1373 0.8691>, rotation=<rot 90 0 0>}]
 dyn_charms_B_1 [{shapeType=Capsule, length=2.4927, radius=0.5, shapeName=dyn_charms_A_1_capsule, center=<1.3623 0 0>, rotation=<rot 90 0 0>}]
 dyn_charms_B_2 [{shapeType=Capsule, length=2.4927, radius=0.5, shapeName=dyn_charms_A_1_capsule, center=<1.3623 0 0>, rotation=<rot 90 0 0>}]
 dyn_charms_B_3 [{shapeType=Capsule, length=7.7623, radius=1.623, shapeName=dyn_charms_A_1_capsule, center=<4.5546 -0.2271 1.2191>, rotation=<rot 90 0 0>}]
 dyn_charms_ring_1 [{
   shapeType=Capsule
   length=2.3692
   radius=0.6087
   shapeName=dyn_charms_ring_1_capsule
   center=<1.677 0.0154 0.2483>
   rotation=<rot -0.3824 -83.6064 -86.5907>
  }]
 dyn_charms_ring_2 [{shapeType=Capsule, length=0.5, radius=0.25, shapeName=dyn_charms_ring_2_capsule}]
 dyn_r_skirt_A_1 [{
   shapeType=Capsule
   length=15.4501
   radius=8.0082
   shapeName=dyn_r_skirt_A_1_capsule
   center=<6.912 0.4566 -6.6436>
   rotation=<rot -82.832 2.4925 1.2793>
  }]
 pelvis [{shapeType=Capsule, length=10.681, radius=8.3151, shapeName=pelvis_capsule, center=<0 -0.277 -0.1226>}]
 spine_01 [{shapeType=Capsule, length=9.0637, radius=8.5758, shapeName=spine_01_capsule, center=<1.2859 0.0489 -1.6857>}]
 spine_02 [{shapeType=Capsule, length=4.6196, radius=8.7332, shapeName=spine_02_capsule, center=<2.3047 1.5372 -1.2812>}]
 thigh_twist_01_r [{
   shapeType=Capsule
   length=14.942
   radius=8.6971
   shapeName=thigh_twist_01_r_capsule
   center=<-20.4159 -0.4235 -2.5052>
   rotation=<rot 83.0017 179.9996 -173.6469>
  }]
}
constraints [
 {bone1Name=dyn_charms_ring_1, bone2Name=base_charms_ring_1, swing1LimitDegrees=0, swing2LimitDegrees=20, twistLimitDegrees=0}
 {bone1Name=dyn_charms_ring_2, bone2Name=dyn_charms_ring_1, swing1LimitDegrees=18, swing2LimitDegrees=0, twistLimitDegrees=0}
 {bone1Name=dyn_charms_A_1, bone2Name=dyn_charms_ring_2, swing1LimitDegrees=20, swing2LimitDegrees=20, twistLimitDegrees=5}
 {bone1Name=dyn_charms_A_2, bone2Name=dyn_charms_A_1, swing1LimitDegrees=20, swing2LimitDegrees=20, twistLimitDegrees=5}
 {bone1Name=dyn_charms_B_1, bone2Name=dyn_charms_ring_2, swing1LimitDegrees=20, swing2LimitDegrees=20, twistLimitDegrees=5}
 {bone1Name=dyn_charms_B_2, bone2Name=dyn_charms_B_1, swing1LimitDegrees=20, swing2LimitDegrees=20, twistLimitDegrees=5}
 {bone1Name=dyn_charms_B_3, bone2Name=dyn_charms_B_2, swing1LimitDegrees=20, swing2LimitDegrees=20, twistLimitDegrees=5}
 {
  bone1Name=base_charms_ring_1
  bone2Name=spine_02
  swing1LimitDegrees=45
  swing1Motion=Free
  swing2LimitDegrees=45
  swing2Motion=Free
  twistLimitDegrees=45
  twistMotion=Free
 }
 {
  bone1Name=spine_01
  bone2Name=pelvis
  swing1LimitDegrees=45
  swing1Motion=Free
  swing2LimitDegrees=45
  swing2Motion=Free
  twistLimitDegrees=45
  twistMotion=Free
 }
 {
  bone1Name=spine_02
  bone2Name=spine_01
  swing1LimitDegrees=45
  swing1Motion=Free
  swing2LimitDegrees=45
  swing2Motion=Free
  twistLimitDegrees=45
  twistMotion=Free
 }
 {
  bone1Name=thigh_twist_01_r
  bone2Name=pelvis
  swing1LimitDegrees=45
  swing1Motion=Free
  swing2LimitDegrees=45
  swing2Motion=Free
  twistLimitDegrees=45
  twistMotion=Free
 }
 {
  bone1Name=dyn_r_skirt_A_1
  bone2Name=pelvis
  swing1LimitDegrees=45
  swing1Motion=Free
  swing2LimitDegrees=45
  swing2Motion=Free
  twistLimitDegrees=45
  twistMotion=Free
 }
]
