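# Physics asset for the center and left skirt bone chains of the AzureBlade medium female player body.
# Six chains of two or three dyn bodies hang from tiny base anchor capsules, using capsules except for boxes on the lower center A segments.
# Constraints link each segment to its parent, with tight swing1 and wider swing2 and twist limit values.
# Unconstrained double-capsule bodies on both thigh twist bones provide leg collision shapes.
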PhysicsAsset /BRCosmetics/Characters/Player/Female/Medium/Bodies/F_MED_AzureBlade/Meshes/F_MED_AzureBlade_Physics_skirt_left
bodies {
 base_c_skirt_A_1 [{shapeType=Capsule, length=1, radius=0.5, shapeName=base_c_skirt_A_1_capsule}]
 base_c_skirt_B_1 [{shapeType=Capsule, length=1, radius=0.5, shapeName=base_c_skirt_B_1_capsule}]
 base_l_skirt_AA_1 [{shapeType=Capsule, length=1, radius=0.5, shapeName=base_l_skirt_A_1_capsule}]
 base_l_skirt_A_1 [{shapeType=Capsule, length=1, radius=0.5, shapeName=base_l_skirt_A_1_capsule}]
 base_l_skirt_B_1 [{shapeType=Capsule, length=1, radius=0.5, shapeName=base_l_skirt_B_1_capsule}]
 base_l_skirt_C_1 [{shapeType=Capsule, length=1, radius=0.5, shapeName=base_l_skirt_C_1_capsule}]
 dyn_c_skirt_A_1 [{
   shapeType=Capsule
   length=4.2397
   radius=6.7386
   shapeName=dyn_c_skirt_A_1_capsule
   center=<6.5 1.4335 2.4692>
   rotation=<rot 90 -165.9638 104.0362>
  }]
 dyn_c_skirt_A_2 [{
   shapeType=Box
   extentX=12.6639
   extentY=11.4204
   extentZ=2.2646
   shapeName=dyn_c_skirt_A_2_box
   center=<7.3762 1.0205 0>
   rotation=<rot 0.0005 -8.3853 -0.0001>
  }]
 dyn_c_skirt_A_3 [{
   shapeType=Box
   extentX=14.8107
   extentY=8.2444
   extentZ=1.9167
   shapeName=dyn_c_skirt_A_3_box
   center=<7.6287 0.1872 0>
   rotation=<rot 0.0004 -9.3534 -0.0001>
  }]
 dyn_c_skirt_B_1 [{shapeType=Capsule, length=10, radius=6.7386, shapeName=dyn_c_skirt_A_1_capsule, center=<6.5 0 1.7664>, rotation=<rot 90 -165.9638 104.0362>}]
 dyn_c_skirt_B_2 [{shapeType=Capsule, length=10, radius=5.8032, shapeName=dyn_c_skirt_A_1_capsule, center=<6.5 0 4.5561>, rotation=<rot 90 0 -59.8828>}]
 dyn_l_skirt_AA_1 [{shapeType=Capsule, length=8.5158, radius=4.6446, shapeName=dyn_c_skirt_A_1_capsule, center=<9.4174 0 2.0752>, rotation=<rot 90 0 0>}]
 dyn_l_skirt_AA_2 [{shapeType=Capsule, length=8.5158, radius=4.6446, shapeName=dyn_c_skirt_A_1_capsule, center=<9.4174 0 2.2712>, rotation=<rot 90 0 0>}]
 dyn_l_skirt_A_1 [{
   shapeType=Capsule
   length=7.434
   radius=6.0321
   shapeName=dyn_c_skirt_A_1_capsule
   center=<7.3891 1.372 3.5618>
   rotation=<rot 81.0351 90.0001 0.0001>
  }]
 dyn_l_skirt_A_2 [{
   shapeType=Capsule
   length=8.5158
   radius=6.0321
   shapeName=dyn_c_skirt_A_1_capsule
   center=<8.1262 1.3901 3.5618>
   rotation=<rot 81.3314 90.0001 0.0001>
  }]
 dyn_l_skirt_B_1 [{shapeType=Capsule, length=10, radius=6.7386, shapeName=dyn_c_skirt_A_1_capsule, center=<6.5 0 3.4579>, rotation=<rot 90 180 90>}]
 dyn_l_skirt_B_2 [{
   shapeType=Capsule
   length=11.4013
   radius=4.9697
   shapeName=dyn_c_skirt_A_1_capsule
   center=<6.5 0 2.8749>
   rotation=<rot 90 -165.9638 104.0362>
  }]
 dyn_l_skirt_C_1 [{shapeType=Capsule, length=10, radius=6.7386, shapeName=dyn_c_skirt_A_1_capsule, center=<6.5 0 1.7664>, rotation=<rot 90 -165.9638 104.0362>}]
 dyn_l_skirt_C_2 [{shapeType=Capsule, length=10, radius=5.8032, shapeName=dyn_c_skirt_A_1_capsule, center=<6.5 0 3.3974>, rotation=<rot 90 -165.9638 104.0362>}]
 thigh_twist_01_l [
  {
   shapeType=Capsule
   length=21.1309
   radius=6.3251
   shapeName=thigh_twist_01_l_capsuleB
   center=<10.2566 -1.2608 1.8295>
   rotation=<rot 85.1971 179.9996 -177.3217>
  }
  {
   shapeType=Capsule
   length=17.5605
   radius=6.3062
   shapeName=thigh_twist_01_l_capsuleA
   center=<-6.0671 0.1004 0.6757>
   rotation=<rot 85.9009 179.4284 178.113>
  }
 ]
 thigh_twist_01_r [
  {
   shapeType=Capsule
   length=20.84
   radius=6.6159
   shapeName=thigh_twist_01_r_capsuleB
   center=<-10.284 0.6728 -1.8318>
   rotation=<rot 85.1971 179.9996 2.6783>
  }
  {
   shapeType=Capsule
   length=17.5605
   radius=6.3062
   shapeName=thigh_twist_01_r_capsuleA
   center=<6.0799 0.3505 -0.9992>
   rotation=<rot 85.9009 179.4284 -1.887>
  }
 ]
}
constraints [
 {bone1Name=dyn_c_skirt_A_1, bone2Name=base_c_skirt_A_1, swing1LimitDegrees=5, swing2LimitDegrees=22, twistLimitDegrees=15}
 {bone1Name=dyn_c_skirt_A_2, bone2Name=dyn_c_skirt_A_1, swing1LimitDegrees=5, swing2LimitDegrees=25, twistLimitDegrees=20}
 {bone1Name=dyn_c_skirt_A_3, bone2Name=dyn_c_skirt_A_2, swing1LimitDegrees=5, swing2LimitDegrees=25, twistLimitDegrees=15}
 {bone1Name=dyn_c_skirt_B_1, bone2Name=base_c_skirt_B_1, swing1LimitDegrees=5, swing2LimitDegrees=10, twistLimitDegrees=15}
 {bone1Name=dyn_c_skirt_B_2, bone2Name=dyn_c_skirt_B_1, swing1LimitDegrees=5, swing2LimitDegrees=20, twistLimitDegrees=15}
 {bone1Name=dyn_l_skirt_A_1, bone2Name=base_l_skirt_A_1, swing1LimitDegrees=5, swing2LimitDegrees=22, twistLimitDegrees=15}
 {bone1Name=dyn_l_skirt_A_2, bone2Name=dyn_l_skirt_A_1, swing1LimitDegrees=5, swing2LimitDegrees=22, twistLimitDegrees=15}
 {bone1Name=dyn_l_skirt_B_1, bone2Name=base_l_skirt_B_1, swing1LimitDegrees=5, swing2LimitDegrees=22, twistLimitDegrees=15}
 {bone1Name=dyn_l_skirt_B_2, bone2Name=dyn_l_skirt_B_1, swing1LimitDegrees=5, swing2LimitDegrees=22, twistLimitDegrees=15}
 {bone1Name=dyn_l_skirt_C_1, bone2Name=base_l_skirt_C_1, swing1LimitDegrees=5, swing2LimitDegrees=10, twistLimitDegrees=15}
 {bone1Name=dyn_l_skirt_C_2, bone2Name=dyn_l_skirt_C_1, swing1LimitDegrees=5, swing2LimitDegrees=22, twistLimitDegrees=15}
 {bone1Name=dyn_l_skirt_AA_1, bone2Name=base_l_skirt_AA_1, swing1LimitDegrees=5, swing2LimitDegrees=22, twistLimitDegrees=15}
 {bone1Name=dyn_l_skirt_AA_2, bone2Name=dyn_l_skirt_AA_1, swing1LimitDegrees=5, swing2LimitDegrees=22, twistLimitDegrees=15}
]
