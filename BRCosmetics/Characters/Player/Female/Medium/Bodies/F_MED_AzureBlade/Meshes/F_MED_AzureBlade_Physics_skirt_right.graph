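# Physics asset for the right side of the F_MED_AzureBlade cosmetic body's skirt, covering three short bone chains (A, B, C) and the right thigh.
# Each dyn skirt bone has a capsule body and a limited swing and twist constraint to its base bone, which carries a tiny capsule.
# Two overlapping capsules on thigh_twist_01_r approximate the leg as an unconstrained body, apparently for skirt collision.
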PhysicsAsset /BRCosmetics/Characters/Player/Female/Medium/Bodies/F_MED_AzureBlade/Meshes/F_MED_AzureBlade_Physics_skirt_right
bodies {
 base_r_skirt_A_1 [{shapeType=Capsule, length=1, radius=0.5, shapeName=base_r_skirt_A_1_capsule}]
 base_r_skirt_B_1 [{shapeType=Capsule, length=1, radius=0.5, shapeName=base_r_skirt_B_1_capsule}]
 base_r_skirt_C_1 [{shapeType=Capsule, length=1, radius=0.5, shapeName=base_r_skirt_C_1_capsule}]
 dyn_r_skirt_A_1 [{shapeType=Capsule, length=14.546, radius=2.6278, shapeName=dyn_c_skirt_A_1_capsule, center=<9.9639 0 0.9725>, rotation=<rot 90 0 0>}]
 dyn_r_skirt_B_1 [{shapeType=Capsule, length=14.6743, radius=2.4995, shapeName=dyn_c_skirt_A_1_capsule, center=<9.9639 0 0.9725>, rotation=<rot 90 0 0>}]
 dyn_r_skirt_C_1 [{shapeType=Capsule, length=14.6664, radius=3.1416, shapeName=dyn_c_skirt_A_1_capsule, center=<9.9639 -1.7352 0.9725>, rotation=<rot 90 0 0>}]
 thigh_twist_01_r [
  {
   shapeType=Capsule
   length=21.1309
   radius=6.3251
   shapeName=thigh_twist_01_r_capsuleB
   center=<-10.2566 1.2608 -1.8295>
   rotation=<rot 85.1971 179.9996 2.6783>
  }
  {
   shapeType=Capsule
   length=17.5605
   radius=6.3062
   shapeName=thigh_twist_01_r_capsuleA
   center=<6.0799 0.3505 -0.9992>
   rotation=<rot 85.9009 179.4284 -1.887>
  }
 ]
}
constraints [
 {bone1Name=dyn_r_skirt_A_1, bone2Name=base_r_skirt_A_1, swing1LimitDegrees=5, swing2LimitDegrees=22, twistLimitDegrees=15}
 {bone1Name=dyn_r_skirt_B_1, bone2Name=base_r_skirt_B_1, swing1LimitDegrees=5, swing2LimitDegrees=22, twistLimitDegrees=15}
 {bone1Name=dyn_r_skirt_C_1, bone2Name=base_r_skirt_C_1, swing1LimitDegrees=5, swing2LimitDegrees=22, twistLimitDegrees=15}
]
