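# Control Rig for the AzureBlade female face accessory that corrects base hair and braid bones as the body poses.
# DeltaRotation helper turns head and clavicle angles into weights, alongside spherical pose readers on the head, to drive additive ModifyTransforms offsets.
# Public targeting, pistol, crouch and pawn-yaw inputs gate some corrections, with the crouch one also scaled by pelvis height and head pitch.
# A debug-only function, disabled here, copies test controls onto head, neck and arm bones.
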ControlRigBlueprint /BRCosmetics/Characters/Player/Female/Medium/Bodies/F_MED_AzureBlade/Meshes/Parts/F_MED_AzureBlade_FaceAcc_CtrlRig
graphs [
 Multiply_1_2_ContainedGraph {nodes [
   n1 = RigVMFunctionEntryNode()
   n2 = &RigVMFunction_MathDoubleMul(A=@n1.A, B=@n1.B)
   RigVMFunctionReturnNode(Result=&RigVMFunction_MathDoubleMul(B=@n1.D, A=&RigVMFunction_MathDoubleMul(B=@n1.C, A=@n2.Result).Result).Result)
  ]}
 Multiply_3_ContainedGraph {nodes [
   n1 = RigVMFunctionEntryNode()
   Multiply = &RigVMFunction_MathDoubleMul(A=@n1.A, B=@n1.B)
   RigVMFunctionReturnNode(Result=&RigVMFunction_MathDoubleMul(B=@n1.C, A=@Multiply.Result).Result)
  ]}
 New Function_ContainedGraph {nodes [
   n1 = RigVMFunctionEntryNode()
   RigVMFunction_MathBoolToInteger = &RigVMFunction_MathBoolToInteger(Value=@n1.onlyTwist)
   RigVMFunction_MathQuaternionSwingTwist = &RigVMFunction_MathQuaternionSwingTwist(
    TwistAxis=@n1.twistAxis
    Input=&RigUnit_GetRelativeTransformForItem(
     Child=(Type=Bone,Name="None")
     Parent=(Type=Bone,Name="None")
     bChildInitial=False
     bParentInitial=False
     Child.Name=@n1.child
     Parent.Name=@n1.parent
     Parent.Type=@n1.parentType
    ).RelativeTransform.Rotation
   )
   RigVMFunction_MathQuaternionToEuler = &RigVMFunction_MathQuaternionToEuler(
    Value=RigVMDispatchNode(
     Values=((),())
     Index=@RigVMFunction_MathBoolToInteger.Result
     Values.0=@RigVMFunction_MathQuaternionSwingTwist.Swing
     Values.1=@RigVMFunction_MathQuaternionSwingTwist.Twist
    ) {template="DISPATCH_RigVMDispatch_SelectInt32(in Index,in Values,out Result)"}.Result
    RotationOrder=@n1.Rotation Order
   )
   > DISPATCH_RigVMDispatch_Print = RigVMDispatchNode(ScreenColor=(R=1,G=1,B=1,A=1), ScreenDuration=1.5, Enabled=@n1.debugRotation, Value=@RigVMFunction_MathQuaternionToEuler.Result) {template="DISPATCH_RigVMDispatch_Print(in Prefix,in Value,in Enabled,in ScreenDuration,in ScreenColor)"}
   Clamp = &RigVMFunction_MathIntClamp(Value=@n1.primAxis)
   DISPATCH_RigVMDispatch_SelectInt32 = RigVMDispatchNode(
    Values=(0,0,0)
    Index=@Clamp.Result
    Values.0=@RigVMFunction_MathQuaternionToEuler.Result.X
    Values.1=@RigVMFunction_MathQuaternionToEuler.Result.Y
    Values.2=@RigVMFunction_MathQuaternionToEuler.Result.Z
   ) {template="DISPATCH_RigVMDispatch_SelectInt32(in Index,in Values,out Result)"}
   Remap = &RigVMFunction_MathDoubleRemap(
    TargetMaximum=1
    TargetMinimum=0
    bClamp=False
    Value=@DISPATCH_RigVMDispatch_SelectInt32.Result
    SourceMaximum=@n1.max
    SourceMinimum=@n1.min
   )
   Multiply = &RigVMFunction_MathDoubleMul(A=&RigVMFunction_MathDoubleClamp(Value=@Remap.Result).Result, B=@n1.premult)
   > RigVMFunctionReturnNode(outputWeight=@Multiply.Result)
  ]}
 New Function_ContainedGraph {nodes [
   n1 = RigVMFunctionEntryNode()
   RigVMFunction_ControlFlowBranch = &RigVMFunction_ControlFlowBranch(ExecuteContext=@n1.ExecuteContext, Condition=@n1.debug)
   RigVMFunctionReturnNode(ExecuteContext=@RigVMFunction_ControlFlowBranch.Completed)
   n2 = &RigUnit_ItemArray(Items=((Type=Bone,Name="head"),(Type=Bone,Name="neck_02"),(Type=Bone,Name="neck_01"),(Type=Bone,Name="clavicle_l"),(Type=Bone,Name="clavicle_r"),(Type=Bone,Name="upperarm_l"),(Type=Bone,Name="upperarm_r")))
   DISPATCH_RigVMDispatch_ArrayIterator = RigVMDispatchNode(Array=@n2.Items, ExecuteContext=@RigVMFunction_ControlFlowBranch.True) {template="DISPATCH_RigVMDispatch_ArrayIterator(in Array,out Element,out Index,out Count,out Ratio)"}
   Concat = /Script/RigVM.RigVMFunction_NameConcat(B=_ctrl, A=@DISPATCH_RigVMDispatch_ArrayIterator.Element.Name) {is=struct RigVMUnitNode, template="Concat::Execute(in A,in B,out Result)"}
   > &RigUnit_SetTransform(
    Item=(Type=Bone,Name="spine_01")
    Weight=1
    bInitial=False
    Item.Name=@DISPATCH_RigVMDispatch_ArrayIterator.Element.Name
    Value=&RigUnit_GetTransform(Item=(Type=Control,Name="spine_01_ctrl"), bInitial=False, Item.Name=@Concat.Result).Transform
   )
  ]}
 New Function_ContainedGraph {nodes [
   RigVMFunctionEntryNode()
   > RigVMFunction_Sequence = RigVMAggregateNode()
   DeltaRotation = RigVMFunctionReferenceNode(
    Rotation Order=ZYX
    child=head
    max=-50
    parent=head_null
    parentType=Null
    premult=1
    primAxis=2
    twistAxis=(X=1,Y=0,Z=0)
    ExecuteContext=@RigVMFunction_Sequence.A
   )
   > &RigUnit_ModifyTransforms(
    ItemToModify="((Item=(Type=Bone,Name=\"base_hair_C_r_1\"),Transform=(Rotation=(X=-0.0179,Y=0.0451,Z=0.1452,W=0.9882),Translation=(X=-2.0677,Y=-0.9559,Z=0.2207),Scale3D=(X=1,Y=1,Z=1))),(Item=(Type=Bone,Name=\"base_braid_l_1\"),Transform=(Rotation=(X=0,Y=0.2763,Z=0,W=0.9611), ...(truncated)"
    WeightMinimum=0
    Weight=@DeltaRotation.outputWeight
   )
   > &RigUnit_ModifyTransforms(
    ItemToModify=((Item=(Type=Bone,Name="base_hair_C_r_1"),Transform=(Rotation=(X=0,Y=-0.1045,Z=0,W=0.9945),Translation=(X=0,Y=0,Z=0),Scale3D=(X=1,Y=1,Z=1))))
    WeightMinimum=0
    Weight=@DeltaRotation.outputWeight
   )
   > &RigUnit_ModifyTransforms(
    ItemToModify="((Item=(Type=Bone,Name=\"base_hair_Bk_l_1\"),Transform=(Rotation=(X=0,Y=-0.2164,Z=0,W=0.9763),Translation=(X=-0.8,Y=0,Z=0),Scale3D=(X=1,Y=1,Z=1))),(Item=(Type=Bone,Name=\"base_hair_Bk_r_1\"),Transform=(Rotation=(X=0,Y=-0.2164,Z=0,W=0.9763),Translation=(X=-0.8, ...(truncated)"
    WeightMinimum=0
    Weight=@DeltaRotation.outputWeight
   )
   > n1 = &RigUnit_SphericalPoseReader(
    ActiveRegionScaleFactors=(PositiveWidth=1.5,NegativeWidth=1.5,PositiveHeight=1,NegativeHeight=1)
    ActiveRegionSize=0.05
    Debug=(bDrawDebug=false,bDraw2D=False,bDrawLocalAxes=False,DebugScale=25,DebugSegments=20,DebugThickness=0.25)
    DriverAxis=(X=1,Y=0,Z=0)
    DriverItem=(Type=Bone,Name="head")
    FalloffRegionScaleFactors=(PositiveWidth=1.5,NegativeWidth=1.5,PositiveHeight=1,NegativeHeight=1)
    FalloffSize=0.15
    FlipHeightScaling=False
    FlipWidthScaling=False
    OptionalParentItem=(Type=Bone,Name="spine_05")
    RotationOffset=(X=-45,Y=-3,Z=65)
   )
   > &RigUnit_ModifyTransforms(
    ItemToModify="((Item=(Type=Bone,Name=\"base_hair_C_r_1\"),Transform=(Rotation=(X=-0.0468,Y=-0.0258,Z=0.0143,W=0.9985),Translation=(X=-1.8966,Y=-0.8594,Z=-2.6882),Scale3D=(X=1,Y=1,Z=1))),(Item=(Type=Bone,Name=\"base_hair_Bk_r_1\"),Transform=(Rotation=(X=-0.0581,Y=-0.1205,Z=0 ...(truncated)"
    WeightMinimum=0
    Weight=@n1.OutputParam
   )
   > n2 = &RigUnit_SphericalPoseReader(
    ActiveRegionScaleFactors=(PositiveWidth=1,NegativeWidth=1,PositiveHeight=1,NegativeHeight=1)
    ActiveRegionSize=0.05
    Debug=(bDrawDebug=false,bDraw2D=False,bDrawLocalAxes=False,DebugScale=25,DebugSegments=20,DebugThickness=0.25)
    DriverAxis=(X=1,Y=0,Z=0)
    DriverItem=(Type=Bone,Name="head")
    FalloffRegionScaleFactors=(PositiveWidth=1,NegativeWidth=1,PositiveHeight=1,NegativeHeight=1)
    FalloffSize=0.15
    FlipHeightScaling=False
    FlipWidthScaling=False
    OptionalParentItem=(Type=Bone,Name="spine_05")
    RotationOffset=(X=-55,Y=40,Z=50)
   )
   Multiply = RigVMAggregateNode(A=@n2.OutputParam, C=RigVMVariableNode() {variable=pub isUsingPistol}.Value, B=RigVMVariableNode() {variable=pub isTargeting}.Value)
   > &RigUnit_ModifyTransforms(
    ItemToModify="((Item=(Type=Bone,Name=\"base_hair_Bk_r_1\"),Transform=(Rotation=(X=0.1406,Y=-0.4375,Z=-0.1811,W=0.8695),Translation=(X=-1,Y=0,Z=0),Scale3D=(X=1,Y=1,Z=1))),(Item=(Type=Bone,Name=\"base_hair_Bk_l_1\"),Transform=(Rotation=(X=-0.0023,Y=-0.1483,Z=-0.0643,W=0.9868) ...(truncated)"
    WeightMinimum=0
    Weight=@Multiply.Result
   )
   > DeltaRotation_1 = RigVMFunctionReferenceNode(
    Rotation Order=ZYX
    child=clavicle_r
    debugRotation=false
    max=-20
    parent=clavicle_r_null
    parentType=Null
    premult=1
    primAxis=1
    twistAxis=(X=1,Y=0,Z=0)
   )
   > &RigUnit_ModifyTransforms(
    ItemToModify=((Item=(Type=Bone,Name="base_hair_C_r_1"),Transform=(Rotation=(X=0,Y=-0.0872,Z=0,W=0.9962),Translation=(X=-1.5,Y=0,Z=0),Scale3D=(X=1,Y=1,Z=1))))
    WeightMinimum=0
    Weight=&RigVMFunction_MathDoubleMul(A=@DeltaRotation_1.outputWeight, B=RigVMVariableNode() {variable=pub isTargeting}.Value).Result
   )
   RigVMFunctionReturnNode(ExecuteContext=@RigVMFunction_Sequence.C)
   RigUnit_GetRelativeTransformForItem = &RigUnit_GetRelativeTransformForItem(Child=(Type=Bone,Name="pelvis"), Parent=(Type=Bone,Name="root"), bChildInitial=False, bParentInitial=False)
   Remap = &RigVMFunction_MathDoubleRemap(
    SourceMaximum=30
    SourceMinimum=60
    TargetMaximum=0
    TargetMinimum=1
    bClamp=false
    Value=@RigUnit_GetRelativeTransformForItem.RelativeTransform.Translation.Z
   )
   n3 = &RigVMFunction_AnimEvalRichCurve(
    Curve="(EditorCurveData=(Keys=((InterpMode=RCIM_Cubic,TangentMode=RCTM_Break,TangentWeightMode=RCTWM_WeightedNone,Time=0,Value=0,ArriveTangent=0,ArriveTangentWeight=0,LeaveTangent=0,LeaveTangentWeight=0),(InterpMode=RCIM_Cubic,TangentMode=RCTM_Break,TangentWeight ...(truncated)"
    SourceMinimum=0
    TargetMinimum=0
    Value=@Remap.Result
   )
   RigVMVariableNode(Value=@n3.Result, ExecuteContext=@RigVMFunction_Sequence.B) {variable=CloseGroundRemap}
   > DeltaRotation_3 = RigVMFunctionReferenceNode(
    Rotation Order=ZYX
    child=head
    debugRotation=false
    max=-89
    min=-50
    onlyTwist=false
    parent=neck_01
    parentType=Bone
    premult=1
    primAxis=2
    twistAxis=(X=1,Y=0,Z=0)
   )
   > RigVMVariableNode(Value=@DeltaRotation_3.outputWeight) {variable=isHeadUpExtreme}
   Absolute = &RigVMFunction_MathDoubleAbs(Value=RigVMVariableNode() {variable=pub pawnDeltaRotation}.Value.Yaw)
   Multiply_1 = RigVMAggregateNode(
    C=RigVMVariableNode() {variable=pub isCrouching}.Value
    A=RigVMVariableNode() {variable=CloseGroundRemap}.Value
    B=RigVMVariableNode() {variable=isHeadUpExtreme}.Value
    D=&RigVMFunction_MathDoubleRemap(SourceMaximum=4, SourceMinimum=0, TargetMaximum=0, TargetMinimum=1, Value=@Absolute.Result).Result
   )
   > &RigUnit_ModifyTransforms(
    ItemToModify="((Item=(Type=Bone,Name=\"base_hair_Bk_r_1\"),Transform=(Rotation=(X=-0.065,Y=-0.459,Z=0.0853,W=0.8819),Translation=(X=-2.5553,Y=0.0247,Z=0.3856),Scale3D=(X=1,Y=1,Z=1))),(Item=(Type=Bone,Name=\"base_hair_Bk_l_1\"),Transform=(Rotation=(X=0.0363,Y=-0.0531,Z=-0.02 ...(truncated)"
    WeightMinimum=0
    Weight=@Multiply_1.Result
   )
  ]}
 RigVMFunctionLibrary {nodes [
   DeltaRotation = RigVMCollapseNode(ExecuteContext=(), Rotation Order=ZYX, premult=1, twistAxis=(X=1,Y=0,Z=0))
   controllers = RigVMCollapseNode(ExecuteContext=(), debug=False)
   hair = RigVMCollapseNode(ExecuteContext=())
  ]}
 RigVMFunction_Sequence_1_ContainedGraph {nodes [
   n1 = RigVMFunctionEntryNode()
   RigVMFunction_Sequence = &RigVMFunction_Sequence(ExecuteContext=@n1.ExecuteContext)
   n2 = &RigVMFunction_Sequence(ExecuteContext=@RigVMFunction_Sequence.B)
   RigVMFunctionReturnNode(A=@RigVMFunction_Sequence.A, B=@n2.A, C=@n2.B)
  ]}
 RigVMModel {nodes [
   RigUnit_BeginExecution = &RigUnit_BeginExecution()
   controllers = RigVMFunctionReferenceNode(debug=false, ExecuteContext=@RigUnit_BeginExecution.ExecutePin)
   > hair = RigVMFunctionReferenceNode()
  ]}
]
hierarchy [
 root {loc=<0 0 0>, rot=<0 0 0 1>}
 pelvis {parent=root, loc=<0 -0.5301 98.0571>, rot=<0.0127 0.707 -0.0127 -0.707>}
 spine_01 {parent=pelvis, loc=<6.1208 -0.9328 0>, rot=<0 0 0.034 -0.9994>}
 spine_02 {parent=spine_01, loc=<6.5997 -0.2095 0>, rot=<0 0 0.0238 -0.9997>}
 spine_03 {parent=spine_02, loc=<6.4914 -0.1237 0>, rot=<0 0 0.0149 -0.9999>}
 spine_04 {parent=spine_03, loc=<6.9598 0.0794 0>, rot=<0 0 -0.0132 -0.9999>}
 spine_05 {parent=spine_04, loc=<6.9987 0 0>, rot=<0 0 -0.0015 -1>}
 neck_01 {parent=spine_05, loc=<11.3138 0.1529 0>, rot=<0 0 -0.2003 -0.9797>}
 neck_02 {parent=neck_01, loc=<4.5627 0 0>, rot=<0 0 0.0504 -0.9987>}
 head {parent=neck_02, loc=<5.3914 0 0>, rot=<0 0 0.115 -0.9934>}
 base_bang_c_1 {parent=head, loc=<11.8832 6.6805 0.0013>, rot=<0.1956 0.6792 0.6795 0.1968>}
 dyn_bang_c_1 {parent=base_bang_c_1, loc=<0 0 0>, rot=<0 0 0 1>}
 base_bang_l_1 {parent=head, loc=<11.6686 5.7466 -2.8248>, rot=<-0.1484 -0.8201 -0.5203 -0.1862>}
 dyn_bang_l_1 {parent=base_bang_l_1, loc=<0 0 0>, rot=<0 0 0 1>}
 base_bang_r_1 {parent=head, loc=<11.6687 5.7466 2.8248>, rot=<0.1862 0.5203 0.8201 0.1484>}
 dyn_bang_r_1 {parent=base_bang_r_1, loc=<0 0 0>, rot=<0 0 0 1>}
 base_hair_A_l_1 {parent=head, loc=<7.4317 5.0666 -6.4497>, rot=<-0.0019 -0.992 -0.1242 -0.022>}
 dyn_hair_A_l_1 {parent=base_hair_A_l_1, loc=<0 0 0>, rot=<0 0 0 1>}
 dyn_hair_A_l_2 {parent=dyn_hair_A_l_1, loc=<5.918 0 0>, rot=<0.0034 -0.1056 -0.032 -0.9939>}
 base_hair_B_l_1 {parent=head, loc=<6.3035 -0.6358 -6.4852>, rot=<0 -0.9994 0.0345 0>}
 dyn_hair_B_l_1 {parent=base_hair_B_l_1, loc=<0 0 0>, rot=<0 0 0 1>}
 dyn_hair_B_l_2 {parent=dyn_hair_B_l_1, loc=<7.0143 -0.2502 0.4931>, rot=<0 0 0 1>}
 dyn_hair_B_l_4 {parent=dyn_hair_B_l_2, loc=<3.269 3.2092 0.9805>, rot=<0.1312 0.0252 -0.0479 -0.9899>}
 dyn_hair_B_l_3 {parent=dyn_hair_B_l_2, loc=<3.3288 -0.8437 0.9889>, rot=<-0.0537 0.0165 -0.007 -0.9984>}
 base_hair_B_r_1 {parent=head, loc=<7.6987 3.0346 7.0481>, rot=<0.0079 0.0428 0.9976 0.0543>}
 dyn_hair_B_r_1 {parent=base_hair_B_r_1, loc=<0 0 0>, rot=<0 0 0 1>}
 dyn_hair_B_r_2 {parent=dyn_hair_B_r_1, loc=<5.831 0 0>, rot=<-0.0089 -0.1021 0.086 -0.991>}
 base_hair_A_r_1 {parent=head, loc=<7.8293 5.6564 6.143>, rot=<0.0286 0.3031 0.9511 0.0515>}
 dyn_hair_A_r_1 {parent=base_hair_A_r_1, loc=<0 0 0>, rot=<0 0 0 1>}
 dyn_hair_A_r_2 {parent=dyn_hair_A_r_1, loc=<4.847 0 0>, rot=<-0.0113 -0.0791 0.1407 -0.9868>}
 base_hair_C_r_1 {parent=head, loc=<6.7533 -2.0937 7.6468>, rot=<0 0 1 0>}
 dyn_hair_C_r_1 {parent=base_hair_C_r_1, loc=<0 0 0>, rot=<0 0 0 1>}
 dyn_hair_C_r_2 {parent=dyn_hair_C_r_1, loc=<7.8392 0 0>, rot=<0 0 0 1>}
 base_braid_r_1 {parent=head, loc=<3.3593 1.7217 6.6647>, rot=<-0.013 -0.707 -0.7062 -0.0351>}
 dyn_braid_r_1 {parent=base_braid_r_1, loc=<0 0 0>, rot=<0 0 0 1>}
 dyn_braid_r_2 {parent=dyn_braid_r_1, loc=<3.9772 0 0>, rot=<0 0 0 1>}
 dyn_braid_r_3 {parent=dyn_braid_r_2, loc=<3.3756 -0.0353 0.0716>, rot=<0.0005 -0.0276 -0.0191 -0.9994>}
 base_braid_l_1 {parent=head, loc=<2.6644 3.1692 -6.3995>, rot=<-0.0203 -0.74 -0.6714 -0.0361>}
 dyn_braid_l_1 {parent=base_braid_l_1, loc=<0 0 0>, rot=<0 0 0 1>}
 dyn_braid_l_2 {parent=dyn_braid_l_1, loc=<6.8636 0 0>, rot=<0 0.0396 0.0106 -0.9992>}
 dyn_braid_l_3 {parent=dyn_braid_l_2, loc=<6.7587 0 0>, rot=<0.0101 0.0234 -0.004 -0.9997>}
 dyn_braid_l_4 {parent=dyn_braid_l_3, loc=<6.136 0 0>, rot=<-0.0149 0.0755 -0.0002 -0.997>}
 FACIAL_C_FacialRoot {parent=head, loc=<1.9362 0.8413 0>, rot=<0.5021 -0.4979 -0.5021 -0.4979>}
 FACIAL_R_CheekInner {parent=FACIAL_C_FacialRoot, loc=<-2.0855 1.0625 6.7327>, rot=<-0.1233 0.088 0.1972 -0.9686>}
 FACIAL_C_Nose {parent=FACIAL_C_FacialRoot, loc=<0 0.445 8.403>, rot=<-0.1305 0 0 -0.9914>}
 FACIAL_C_NoseLower {parent=FACIAL_C_Nose, loc=<0 2.5009 -0.1479>, rot=<0.1305 0 0 -0.9914>}
 FACIAL_R_Nostril {parent=FACIAL_C_Nose, loc=<-1.0323 1.9182 -0.3606>, rot=<0.0173 -0.1305 -0.0023 -0.9913>}
 FACIAL_C_NoseTip {parent=FACIAL_C_Nose, loc=<0 1.6414 0.6331>, rot=<0.0872 0 0 -0.9962>}
 base_hair_Bk_l_1 {parent=head, loc=<6.8518 -6.566 -3.3681>, rot=<0 -0.8647 0.5023 0>}
 dyn_hair_Bk_l_1 {parent=base_hair_Bk_l_1, loc=<0 0 0>, rot=<0 0 0 1>}
 dyn_hair_Bk_l_2 {parent=dyn_hair_Bk_l_1, loc=<6.9629 0 0>, rot=<0 0.0218 0 -0.9998>}
 base_hair_Bk_r_1 {parent=head, loc=<6.8523 -6.566 3.3681>, rot=<0 -0.5023 0.8647 0>}
 dyn_hair_Bk_r_1 {parent=base_hair_Bk_r_1, loc=<0 0 0>, rot=<0 0 0 -1>}
 dyn_hair_Bk_r_2 {parent=dyn_hair_Bk_r_1, loc=<6.9632 0 0>, rot=<0 0.0218 0 -0.9998>}
 clavicle_l {parent=spine_05, loc=<5.4831 -0.1053 -3.2946>, rot=<0.0465 -0.7289 -0.009 -0.6829>}
 upperarm_l {parent=clavicle_l, loc=<9.8658 -1.2136 -0.4105>, rot=<0.0164 -0.3245 -0.0269 -0.9454>}
 upperarm_twist_01_l {parent=upperarm_l, loc=<7.2135 -0.1134 0>, rot=<0 0 0 1>}
 coll_braid_l_blocker {parent=clavicle_l, loc=<6.9293 0.1937 4.9036>, rot=<0.9994 -0.0001 0.0335 0.0017>}
 clavicle_r {parent=spine_05, loc=<5.4831 -0.1053 3.2946>, rot=<0.7289 0.0465 -0.6829 0.009>}
 upperarm_r {parent=clavicle_r, loc=<-9.8658 1.2136 0.4105>, rot=<0.0164 -0.3245 -0.0269 -0.9454>}
 upperarm_twist_01_r {parent=upperarm_r, loc=<-7.2135 0.1134 0>, rot=<0 0 0 -1>}
 head_null {parent=spine_05, type=Null, loc=<20.6573 3.5489 0>, rot=<0 0 -0.036 -0.9994>}
 clavicle_r_null {parent=spine_05, type=Null, loc=<5.483 -0.1053 3.2946>, rot=<0.7289 0.0465 -0.6829 0.009>}
 clavicle_r_ctrl {controlType=EULER_TRANSFORM, type=Control, loc=<0 0 0>, rot=<0 0 0 1>}
 upperarm_r_ctrl {controlType=EULER_TRANSFORM, parent=clavicle_r_ctrl, type=Control, loc=<0 0 0>, rot=<0 0 0 1>}
 clavicle_l_ctrl {controlType=EULER_TRANSFORM, type=Control, loc=<0 0 0>, rot=<0 0 0 1>}
 upperarm_l_ctrl {controlType=EULER_TRANSFORM, parent=clavicle_l_ctrl, type=Control, loc=<0 0 0>, rot=<0 0 0 1>}
 neck_01_ctrl {controlType=EULER_TRANSFORM, type=Control, loc=<0 0 0>, rot=<0 0 0 1>}
 neck_02_ctrl {controlType=EULER_TRANSFORM, parent=neck_01_ctrl, type=Control, loc=<0 0 0>, rot=<0 0 0 1>}
 head_ctrl {controlType=EULER_TRANSFORM, parent=neck_02_ctrl, type=Control, loc=<0 0 0>, rot=<0 0 0 1>}
]
variables [
 CloseGroundRemap {type=double, default=0}
 isHeadUpExtreme {type=double, default=0}
 pub isCrouching {type=double, default=0}
 pub isTargeting {type=double, default=0}
 pub isUsingPistol {type=double, default=0}
 pub pawnDeltaRotation {type=FRotator, default=(Pitch=0,Yaw=0,Roll=0)}
]
