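# Physics asset for the AzureBlade female medium face accessory part, providing secondary motion setup for braids and hair strands.
# Capsule chains, plus one box, run from base bones at neck_02 through dyn_ bones with tight swing limits and little or no twist.
# Spine, neck, clavicle and upper-arm capsules form body collision volumes, with an oversized blocker capsule off the left clavicle for the braid.
# Base and body links are angularly free, as is a cross-link between back-hair chains.
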PhysicsAsset /BRCosmetics/Characters/Player/Female/Medium/Bodies/F_MED_AzureBlade/Meshes/Parts/F_MED_AzureBlade_FaceAcc_Physics
bodies {
 base_braid_l_1 [{shapeType=Capsule, length=1, radius=0.5, shapeName=base_braid_l_1_capsule}]
 base_braid_r_1 [{shapeType=Capsule, length=1, radius=0.5, shapeName=base_braid_l_1_capsule}]
 base_hair_A_l_1 [{shapeType=Capsule, length=0.505, radius=0.505, shapeName=base_hair_A_l_1_capsule, rotation=<rot 0 -90 0>}]
 base_hair_B_l_1 [{shapeType=Capsule, length=1, radius=0.5, shapeName=base_braid_l_1_capsule}]
 base_hair_B_r_1 [{shapeType=Capsule, length=1, radius=0.5, shapeName=base_braid_l_1_capsule}]
 base_hair_Bk_l_1 [{shapeType=Capsule, length=1, radius=0.5, shapeName=base_braid_l_1_capsule}]
 base_hair_Bk_r_1 [{shapeType=Capsule, length=1, radius=0.5, shapeName=base_braid_l_1_capsule}]
 base_hair_C_r_1 [{shapeType=Capsule, length=1, radius=0.5, shapeName=base_braid_l_1_capsule}]
 clavicle_l [{
   shapeType=Capsule
   length=8.2393
   radius=5.6045
   shapeName=clavicle_l_capsule
   center=<3.5699 -0.8787 -0.1542>
   rotation=<rot -0.0104 -89.9993 -82.3553>
  }]
 clavicle_r [{
   shapeType=Capsule
   length=8.2393
   radius=5.6045
   shapeName=clavicle_r_capsule
   center=<-3.5699 0.8787 0.1542>
   rotation=<rot -0.0104 -89.9993 -82.3553>
  }]
 coll_braid_l_blocker [{
   shapeType=Capsule
   length=16.5225
   radius=11.5882
   shapeName=coll_braid_l_blocker_capsule
   center=<-0.9905 13.0341 -0.3073>
   rotation=<rot 0 -90 -6.7069>
  }]
 dyn_braid_l_1 [{shapeType=Capsule, length=3.3964, radius=0.9661, shapeName=dyn_braid_l_1_capsule, center=<3.2477 0 0>, rotation=<rot 0 90 -90>}]
 dyn_braid_l_2 [{shapeType=Capsule, length=5.5697, radius=0.9661, shapeName=dyn_braid_l_1_capsule, center=<3.7086 0 0>, rotation=<rot 0 90 -90>}]
 dyn_braid_l_3 [{shapeType=Capsule, length=5.2362, radius=0.9661, shapeName=dyn_braid_l_1_capsule, center=<3.1697 0 0>, rotation=<rot 0 90 -90>}]
 dyn_braid_l_4 [{shapeType=Capsule, length=2.2058, radius=0.7162, shapeName=dyn_braid_l_1_capsule, center=<2.029 0 0>, rotation=<rot 0 90 -90>}]
 dyn_braid_r_1 [{
   shapeType=Capsule
   length=1.7081
   radius=0.9172
   shapeName=dyn_braid_r_1_capsule
   center=<1.2175 -0.0124 0.0066>
   rotation=<rot 0 87.1576 -86.4593>
  }]
 dyn_braid_r_2 [{
   shapeType=Capsule
   length=2.8527
   radius=0.8191
   shapeName=dyn_braid_r_2_capsule
   center=<1.2605 0.089 0.214>
   rotation=<rot 0 93.5139 -88.448>
  }]
 dyn_braid_r_3 [{shapeType=Capsule, length=2.4069, radius=0.8031, shapeName=dyn_braid_r_3_capsule, center=<1.9623 0.1478 0.0961>, rotation=<rot 0 -92 -90>}]
 dyn_hair_A_l_1 [{
   shapeType=Capsule
   length=3.8458
   radius=1.2912
   shapeName=dyn_hair_B_l_2_capsule
   center=<3.13 0.1638 0.5584>
   rotation=<rot 0 90.4972 -90.7272>
  }]
 dyn_hair_A_l_2 [{
   shapeType=Capsule
   length=3.8338
   radius=1.2327
   shapeName=dyn_hair_A_l_2_capsule
   center=<3.0273 0 0.508>
   rotation=<rot -85.1201 -179.9992 179.9994>
  }]
 dyn_hair_B_l_1 [{shapeType=Capsule, length=4.9149, radius=2.0589, shapeName=base_braid_l_1_capsule, center=<3.58 0 0.8572>, rotation=<rot 0 0 90>}]
 dyn_hair_B_l_2 [{
   shapeType=Capsule
   length=4.956
   radius=1.5687
   shapeName=dyn_hair_B_l_2_capsule
   center=<1.5252 0.419 1.2707>
   rotation=<rot 0.1766 -1.008 -87.7838>
  }]
 dyn_hair_B_l_3 [{
   shapeType=Capsule
   length=2.6328
   radius=1.6146
   shapeName=dyn_hair_B_l_3_capsule
   center=<2.1819 0.1206 0.2044>
   rotation=<rot -1.1805 176.9857 -83.2564>
  }]
 dyn_hair_B_l_4 [{
   shapeType=Capsule
   length=3.1858
   radius=0.8954
   shapeName=dyn_hair_B_l_4_capsule
   center=<2.3638 0 0.0719>
   rotation=<rot 0 92.4405 -87.6643>
  }]
 dyn_hair_B_r_1 [{
   shapeType=Capsule
   length=3.2297
   radius=1.3482
   shapeName=dyn_hair_B_r_1_capsule
   center=<4.1078 -0.0676 0.1015>
   rotation=<rot 0.1006 -84.6067 -88.4157>
  }]
 dyn_hair_B_r_2 [{
   shapeType=Capsule
   length=2.1757
   radius=1.3758
   shapeName=dyn_hair_B_r_2_capsule
   center=<2.7229 -0.0008 -0.2214>
   rotation=<rot -5.9139 -94.0397 -90.3017>
  }]
 dyn_hair_Bk_l_1 [{shapeType=Capsule, length=3.2335, radius=2.0564, shapeName=dyn_hair_Bk_l_1_capsule, center=<4.0291 0 0>, rotation=<rot 90 0 0>}]
 dyn_hair_Bk_l_2 [{
   shapeType=Capsule
   length=2.6556
   radius=2.8579
   shapeName=dyn_hair_Bk_l_2_capsule
   center=<3.607 -1.1487 1.2679>
   rotation=<rot 0.0001 -180 -74.6491>
  }]
 dyn_hair_Bk_r_1 [{shapeType=Capsule, length=3.2335, radius=2.0564, shapeName=dyn_hair_Bk_l_1_capsule, center=<4.0291 0 0>, rotation=<rot 90 0 0>}]
 dyn_hair_Bk_r_2 [{
   shapeType=Capsule
   length=2.6556
   radius=2.8579
   shapeName=dyn_hair_Bk_l_2_capsule
   center=<3.607 -1.1487 1.2679>
   rotation=<rot 0.0001 -180 -74.6491>
  }]
 dyn_hair_C_r_1 [{shapeType=Capsule, length=4.9149, radius=2.0589, shapeName=base_braid_l_1_capsule, center=<3.58 0 0>, rotation=<rot 0 0 90>}]
 dyn_hair_C_r_2 [{
   shapeType=Box
   extentX=7.1708
   extentY=6.4705
   extentZ=1.7359
   shapeName=dyn_hair_C_r_2_box
   center=<3.6052 -0.0419 0.3047>
   rotation=<rot 5.7982 4.0392 7.1722>
  }]
 neck_01 [{shapeType=Capsule, length=4.6366, radius=4.7106, shapeName=neck_01_capsule, center=<-0.8977 0.1697 0>, rotation=<rot 0 -90.0001 0>}]
 neck_02 [{shapeType=Capsule, length=6.5205, radius=4.4498, shapeName=neck_02_capsule, rotation=<rot 90 0 0>}]
 spine_04 [{shapeType=Capsule, length=7.1939, radius=9.3008, shapeName=spine_05_capsule, center=<2.0994 2.0859 0>, rotation=<rot 0 -90 0>}]
 spine_05 [{shapeType=Capsule, length=10.2044, radius=8.5624, shapeName=spine_05_capsule, center=<-0.0094 0.147 0>, rotation=<rot 0 -90 0>}]
 upperarm_twist_01_l [{shapeType=Capsule, length=19.6714, radius=4.0679, shapeName=upperarm_twist_01_l_capsule, center=<1.8914 0 -0.932>, rotation=<rot 90 0 0>}]
 upperarm_twist_01_r [
  {shapeType=Capsule, length=19.6714, radius=4.0679, shapeName=upperarm_twist_01_l_capsule, center=<-1.8914 0 0.932>, rotation=<rot 90 0 0>}
  {
   shapeType=Capsule
   length=9.5308
   radius=4.6419
   shapeName=upperarm_twist_01_r_capsule
   center=<-1.6479 -6.5912 0.932>
   rotation=<rot 90 0 0>
  }
 ]
}
constraints [
 {bone1Name=dyn_braid_l_1, bone2Name=base_braid_l_1, swing1LimitDegrees=10, swing2LimitDegrees=10, twistLimitDegrees=0}
 {bone1Name=dyn_braid_l_2, bone2Name=dyn_braid_l_1, swing1LimitDegrees=14, swing2LimitDegrees=14, twistLimitDegrees=0}
 {bone1Name=dyn_braid_l_3, bone2Name=dyn_braid_l_2, swing1LimitDegrees=16, swing2LimitDegrees=16, twistLimitDegrees=0}
 {bone1Name=dyn_braid_l_4, bone2Name=dyn_braid_l_3, swing1LimitDegrees=18, swing2LimitDegrees=18, twistLimitDegrees=0}
 {bone1Name=dyn_hair_B_l_1, bone2Name=base_hair_B_l_1, swing1LimitDegrees=15, swing2LimitDegrees=15, twistLimitDegrees=0}
 {bone1Name=dyn_hair_B_l_2, bone2Name=dyn_hair_B_l_1, swing1LimitDegrees=18, swing2LimitDegrees=18, twistLimitDegrees=0}
 {bone1Name=dyn_hair_C_r_1, bone2Name=base_hair_C_r_1, swing1LimitDegrees=10, swing2LimitDegrees=12, twistLimitDegrees=0}
 {bone1Name=dyn_hair_C_r_2, bone2Name=dyn_hair_C_r_1, swing1LimitDegrees=16, swing2LimitDegrees=15, twistLimitDegrees=0}
 {
  bone1Name=base_hair_B_l_1
  bone2Name=neck_02
  swing1LimitDegrees=45
  swing1Motion=Free
  swing2LimitDegrees=45
  swing2Motion=Free
  twistLimitDegrees=45
  twistMotion=Free
 }
 {
  bone1Name=base_hair_C_r_1
  bone2Name=neck_02
  swing1LimitDegrees=45
  swing1Motion=Free
  swing2LimitDegrees=45
  swing2Motion=Free
  twistLimitDegrees=45
  twistMotion=Free
 }
 {
  bone1Name=base_braid_l_1
  bone2Name=neck_02
  swing1LimitDegrees=45
  swing1Motion=Free
  swing2LimitDegrees=45
  swing2Motion=Free
  twistLimitDegrees=45
  twistMotion=Free
 }
 {
  bone1Name=spine_05
  bone2Name=spine_04
  swing1LimitDegrees=45
  swing1Motion=Free
  swing2LimitDegrees=45
  swing2Motion=Free
  twistLimitDegrees=45
  twistMotion=Free
 }
 {
  bone1Name=clavicle_l
  bone2Name=spine_05
  swing1LimitDegrees=45
  swing1Motion=Free
  swing2LimitDegrees=45
  swing2Motion=Free
  twistLimitDegrees=45
  twistMotion=Free
 }
 {
  bone1Name=neck_01
  bone2Name=spine_05
  swing1LimitDegrees=45
  swing1Motion=Free
  swing2LimitDegrees=45
  swing2Motion=Free
  twistLimitDegrees=45
  twistMotion=Free
 }
 {
  bone1Name=neck_02
  bone2Name=neck_01
  swing1LimitDegrees=45
  swing1Motion=Free
  swing2LimitDegrees=45
  swing2Motion=Free
  twistLimitDegrees=45
  twistMotion=Free
 }
 {
  bone1Name=upperarm_twist_01_l
  bone2Name=clavicle_l
  swing1LimitDegrees=45
  swing1Motion=Free
  swing2LimitDegrees=45
  swing2Motion=Free
  twistLimitDegrees=45
  twistMotion=Free
 }
 {bone1Name=dyn_braid_r_2, bone2Name=dyn_braid_r_1, swing1LimitDegrees=12, swing2LimitDegrees=12, twistLimitDegrees=0}
 {bone1Name=dyn_braid_r_1, bone2Name=base_braid_r_1, swing1LimitDegrees=10, swing2LimitDegrees=7, twistLimitDegrees=0}
 {
  bone1Name=base_braid_r_1
  bone2Name=neck_02
  swing1LimitDegrees=45
  swing1Motion=Free
  swing2LimitDegrees=45
  swing2Motion=Free
  twistLimitDegrees=45
  twistMotion=Free
 }
 {bone1Name=dyn_hair_A_l_2, bone2Name=dyn_hair_A_l_1, swing1LimitDegrees=20, swing2LimitDegrees=20, twistLimitDegrees=0}
 {bone1Name=dyn_hair_A_l_1, bone2Name=base_hair_A_l_1, swing1LimitDegrees=15, swing2LimitDegrees=10, twistLimitDegrees=0}
 {
  bone1Name=base_hair_A_l_1
  bone2Name=neck_02
  swing1LimitDegrees=45
  swing1Motion=Free
  swing2LimitDegrees=45
  swing2Motion=Free
  twistLimitDegrees=45
  twistMotion=Free
 }
 {bone1Name=dyn_braid_r_3, bone2Name=dyn_braid_r_2, swing1LimitDegrees=15, swing2LimitDegrees=15, twistLimitDegrees=0}
 {bone1Name=dyn_hair_B_l_3, bone2Name=dyn_hair_B_l_2, swing1LimitDegrees=15, swing2LimitDegrees=10, twistLimitDegrees=0}
 {bone1Name=dyn_hair_B_l_4, bone2Name=dyn_hair_B_l_2, swing1LimitDegrees=15, swing2LimitDegrees=10, twistLimitDegrees=0}
 {
  bone1Name=base_hair_B_r_1
  bone2Name=neck_02
  swing1LimitDegrees=45
  swing1Motion=Free
  swing2LimitDegrees=45
  swing2Motion=Free
  twistLimitDegrees=45
  twistMotion=Free
 }
 {bone1Name=dyn_hair_B_r_1, bone2Name=base_hair_B_r_1, swing1LimitDegrees=15, swing2LimitDegrees=15, twistLimitDegrees=0}
 {bone1Name=dyn_hair_B_r_2, bone2Name=dyn_hair_B_r_1, swing1LimitDegrees=15, swing2LimitDegrees=15, twistLimitDegrees=0}
 {
  bone1Name=coll_braid_l_blocker
  bone2Name=clavicle_l
  swing1LimitDegrees=45
  swing1Motion=Free
  swing2LimitDegrees=45
  swing2Motion=Free
  twistLimitDegrees=45
  twistMotion=Free
 }
 {bone1Name=dyn_hair_Bk_r_2, bone2Name=dyn_hair_Bk_r_1, swing1LimitDegrees=15, swing2LimitDegrees=20, twistLimitDegrees=5}
 {
  bone1Name=base_hair_Bk_l_1
  bone2Name=neck_02
  swing1LimitDegrees=45
  swing1Motion=Free
  swing2LimitDegrees=45
  swing2Motion=Free
  twistLimitDegrees=45
  twistMotion=Free
 }
 {bone1Name=dyn_hair_Bk_l_1, bone2Name=base_hair_Bk_l_1, swing1LimitDegrees=15, swing2LimitDegrees=20, twistLimitDegrees=5}
 {bone1Name=dyn_hair_Bk_l_2, bone2Name=dyn_hair_Bk_l_1, swing1LimitDegrees=15, swing2LimitDegrees=20, twistLimitDegrees=5}
 {
  bone1Name=base_hair_Bk_r_1
  bone2Name=neck_02
  swing1LimitDegrees=45
  swing1Motion=Free
  swing2LimitDegrees=45
  swing2Motion=Free
  twistLimitDegrees=45
  twistMotion=Free
 }
 {bone1Name=dyn_hair_Bk_r_1, bone2Name=base_hair_Bk_r_1, swing1LimitDegrees=15, swing2LimitDegrees=20, twistLimitDegrees=5}
 {
  bone1Name=dyn_hair_Bk_r_1
  bone2Name=dyn_hair_Bk_l_1
  swing1LimitDegrees=45
  swing1Motion=Free
  swing2LimitDegrees=45
  swing2Motion=Free
  twistLimitDegrees=45
  twistMotion=Free
 }
]
