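# Animation Blueprint for the AzureBlade female character's face accessory part, animating hair, braid and nose-area deformation.
# The parent mesh pose feeds a Control Rig driven by crouch, targeting and pistol state, then a braid tweak.
# AnimDynamics and RigidBody simulate hair, with bone constraints posing the left braid when rigid body is disabled.
# Finally, five nose and cheek bones replace their transforms with values read from the Head mesh each update.
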AnimBlueprint /BRCosmetics/Characters/Player/Female/Medium/Bodies/F_MED_AzureBlade/Meshes/Parts/F_MED_AzureBlade_FaceAcc_AnimBP
target_skeleton=&/F_MED_AzureBlade_FaceAcc_Skeleton
graphs [
 AnimGraph {nodes [
   Output Pose(Result=Use cached pose 'NoseDeformationCP'()) {layerGroup=None}
   n1 = Control Rig!(
    Source=Copy Pose From Mesh(SourceMeshComponent=Get SkeletalMesh() -> (SkeletalMesh:Skeletal Mesh Component Object Reference)) {bUseAttachedParent=true}
    Alpha
    pub isTargeting:double=Get isTargeting float() -> (isTargeting float:double)
    pub isUsingPistol:double=Get isUsingPistol() -> (isUsingPistol:double)
    pub isCrouching:double=Get isCrouching float() -> (isCrouching float:double)
    pub pawnDeltaRotation:Rotator=Get PawnDeltaRotation() -> (PawnDeltaRotation:Rotator)
   ) -> (Pose) {
    alphaBoolBlend {blendOption=Linear}
    controlRigAssetReference {blueprintRigClass=&/F_MED_AzureBlade_FaceAcc_CtrlRig.F_MED_AzureBlade_FaceAcc_CtrlRig_C}
   }
   Save cached pose 'controlRigCP'(Pose=@n1) {cacheName=controlRigCP}
   n2 = "Transform (Modify) Bone - Bone: FACIAL_R_CheekInner"!(
    ComponentPose=Local To Component(LocalPose=Use cached pose 'NoRBAN_SetupCP'())
    bAlphaBoolEnabled:bool=True
    Translation=Get FACIAL_R_CheekInner_Translation() -> (FACIAL_R_CheekInner_Translation:Vector)
    Rotation=Get FACIAL_R_CheekInner_Rotation() -> (FACIAL_R_CheekInner_Rotation:Rotator)
    Scale="1,1,1"
   ) -> (Pose) {
    alphaInputType=Bool
    rotationMode=BMM_Replace
    rotationSpace=BCS_ParentBoneSpace
    translationMode=BMM_Replace
    translationSpace=BCS_ParentBoneSpace
    alphaBoolBlend {blendOption=Linear}
    boneToModify {boneName=FACIAL_R_CheekInner}
   }
   n3 = "Transform (Modify) Bone - Bone: FACIAL_C_Nose"!(
    ComponentPose=@n2
    bAlphaBoolEnabled:bool=True
    Translation=Get FACIAL_C_Nose_Translation() -> (FACIAL_C_Nose_Translation:Vector)
    Rotation=Get FACIAL_C_Nose_Rotation() -> (FACIAL_C_Nose_Rotation:Rotator)
    Scale="1,1,1"
   ) -> (Pose) {
    alphaInputType=Bool
    rotationMode=BMM_Replace
    rotationSpace=BCS_ParentBoneSpace
    translationMode=BMM_Replace
    translationSpace=BCS_ParentBoneSpace
    alphaBoolBlend {blendOption=Linear}
    boneToModify {boneName=FACIAL_C_Nose}
   }
   n4 = "Transform (Modify) Bone - Bone: FACIAL_C_NoseLower"!(
    ComponentPose=@n3
    bAlphaBoolEnabled:bool=True
    Translation=Get FACIAL_C_NoseLower_Translation() -> (FACIAL_C_NoseLower_Translation:Vector)
    Rotation=Get FACIAL_C_NoseLower_Rotation() -> (FACIAL_C_NoseLower_Rotation:Rotator)
    Scale="1,1,1"
   ) -> (Pose) {
    alphaInputType=Bool
    rotationMode=BMM_Replace
    rotationSpace=BCS_ParentBoneSpace
    translationMode=BMM_Replace
    translationSpace=BCS_ParentBoneSpace
    alphaBoolBlend {blendOption=Linear}
    boneToModify {boneName=FACIAL_C_NoseLower}
   }
   n5 = "Transform (Modify) Bone - Bone: FACIAL_R_Nostril"!(
    ComponentPose=@n4
    bAlphaBoolEnabled:bool=True
    Translation=Get FACIAL_R_Nostril_Translation() -> (FACIAL_R_Nostril_Translation:Vector)
    Rotation=Get FACIAL_R_Nostril_Rotation() -> (FACIAL_R_Nostril_Rotation:Rotator)
    Scale="1,1,1"
   ) -> (Pose) {
    alphaInputType=Bool
    rotationMode=BMM_Replace
    rotationSpace=BCS_ParentBoneSpace
    translationMode=BMM_Replace
    translationSpace=BCS_ParentBoneSpace
    alphaBoolBlend {blendOption=Linear}
    boneToModify {boneName=FACIAL_R_Nostril}
   }
   n6 = "Transform (Modify) Bone - Bone: FACIAL_C_NoseTip"!(
    ComponentPose=@n5
    bAlphaBoolEnabled:bool=True
    Translation=Get FACIAL_C_NoseTip_Translation() -> (FACIAL_C_NoseTip_Translation:Vector)
    Rotation=Get FACIAL_C_NoseTip_Rotation() -> (FACIAL_C_NoseTip_Rotation:Rotator)
    Scale="1,1,1"
   ) -> (Pose) {
    alphaInputType=Bool
    rotationMode=BMM_Replace
    rotationSpace=BCS_ParentBoneSpace
    translationMode=BMM_Replace
    translationSpace=BCS_ParentBoneSpace
    alphaBoolBlend {blendOption=Linear}
    boneToModify {boneName=FACIAL_C_NoseTip}
   }
   Save cached pose 'NoseDeformationCP'(Pose=Component To Local(ComponentPose=@n6)) {cacheName=NoseDeformationCP}
   n7 = "Transform (Modify) Bone - Bone: base_braid_l_1"!(
    ComponentPose=Local To Component(LocalPose=Use cached pose 'controlRigCP'())
    bAlphaBoolEnabled:bool=True
    Translation="0,0,0"
    Rotation="10,0,0"
   ) -> (Pose) {
    alphaInputType=Bool
    rotationMode=BMM_Additive
    rotationSpace=BCS_BoneSpace
    translationMode=BMM_Additive
    translationSpace=BCS_BoneSpace
    alphaBoolBlend {blendInTime=0.5, blendOption=Cubic, blendOutTime=0.5}
    boneToModify {boneName=base_braid_l_1}
    rotation {pitch=10}
   }
   Save cached pose 'preSimTweaksCP'(Pose=Component To Local(ComponentPose=@n7)) {cacheName=preSimTweaksCP}
   n8 = Anim Dynamics - Bone: dyn_hair_A_r_2!(
    ComponentPose=Local To Component(LocalPose=Use cached pose 'preSimTweaksCP'())
    Alpha=1
    GravityOverride:Vector="0,0,-980"
    ExternalForce="0,0,0"
   ) -> (Pose) {
    angularDampingOverride=1
    angularSpringConstant=8000
    bAngularSpring=true
    bOverrideAngularDamping=true
    bUseGravityOverride=true
    gravityScale=0.35
    simulationSpace=RootRelative
    alphaBoolBlend {blendOption=Linear}
    boundBone {boneName=dyn_hair_A_r_2}
    componentAppliedLinearAccClamp=<3000 3000 3000>
    componentLinearAccScale=<0.3 0.4 0.3>
    componentLinearVelScale=<0.3 0.4 0.3>
    gravityOverride {z=-980}
    physicsBodyDefinitions [{
      boundBone {boneName=dyn_hair_A_r_2}
      boxExtents=<20 30 30>
      constraintSetup {angularLimitsMax=<0 0 15>, angularLimitsMin=<0 -15 -15>, angularTarget=<1 0 0>}
      localJointOffset=<5 0 0>
     }]
   }
   n9 = Anim Dynamics - Bone: dyn_hair_A_r_1!(ComponentPose=@n8, Alpha=1, GravityOverride:Vector="0,0,-980", ExternalForce="0,0,0") -> (Pose) {
    angularDampingOverride=1
    angularSpringConstant=8000
    bAngularSpring=true
    bOverrideAngularDamping=true
    bUseGravityOverride=true
    gravityScale=0.35
    simulationSpace=RootRelative
    alphaBoolBlend {blendOption=Linear}
    boundBone {boneName=dyn_hair_A_r_1}
    componentAppliedLinearAccClamp=<3000 3000 3000>
    componentLinearAccScale=<0.3 0.4 0.3>
    componentLinearVelScale=<0.3 0.4 0.3>
    gravityOverride {z=-980}
    physicsBodyDefinitions [{
      boundBone {boneName=dyn_hair_A_r_1}
      boxExtents=<20 30 30>
      constraintSetup {angularLimitsMax=<0 0 15>, angularLimitsMin=<0 -15 -15>, angularTarget=<1 0 0>}
      localJointOffset=<5 0 0>
     }]
   }
   n10 = Anim Dynamics - Bone: dyn_bang_l_1!(ComponentPose=@n9, Alpha=1, GravityOverride:Vector="0,0,-980", ExternalForce="0,0,0") -> (Pose) {
    angularDampingOverride=1
    angularSpringConstant=10000
    bAngularSpring=true
    bOverrideAngularDamping=true
    bUseGravityOverride=true
    gravityScale=0.35
    simulationSpace=RootRelative
    alphaBoolBlend {blendOption=Linear}
    boundBone {boneName=dyn_bang_l_1}
    componentAppliedLinearAccClamp=<3000 3000 3000>
    componentLinearAccScale=<0.4 0.4 0.4>
    componentLinearVelScale=<0.4 0.4 0.4>
    gravityOverride {z=-980}
    physicsBodyDefinitions [{
      boundBone {boneName=dyn_bang_l_1}
      boxExtents=<20 30 30>
      constraintSetup {angularLimitsMax=<0 0 7>, angularLimitsMin=<0 -15 -7>, angularTarget=<1 0 0>}
      localJointOffset=<4 0 0>
     }]
   }
   n11 = Anim Dynamics - Bone: dyn_bang_c_1!(ComponentPose=@n10, Alpha=1, GravityOverride:Vector="0,0,-980", ExternalForce="0,0,0") -> (Pose) {
    angularDampingOverride=1
    angularSpringConstant=10000
    bAngularSpring=true
    bOverrideAngularDamping=true
    bUseGravityOverride=true
    gravityScale=0.35
    simulationSpace=RootRelative
    alphaBoolBlend {blendOption=Linear}
    boundBone {boneName=dyn_bang_c_1}
    componentAppliedLinearAccClamp=<3000 3000 3000>
    componentLinearAccScale=<0.4 0.4 0.4>
    componentLinearVelScale=<0.4 0.4 0.4>
    gravityOverride {z=-980}
    physicsBodyDefinitions [{
      boundBone {boneName=dyn_bang_c_1}
      boxExtents=<20 30 30>
      constraintSetup {angularLimitsMax=<0 0 7>, angularLimitsMin=<0 -15 -7>, angularTarget=<1 0 0>}
      localJointOffset=<4 0 0>
     }]
   }
   n12 = Anim Dynamics - Bone: dyn_bang_r_1!(ComponentPose=@n11, Alpha=1, GravityOverride:Vector="0,0,-980", ExternalForce="0,0,0") -> (Pose) {
    angularDampingOverride=1
    angularSpringConstant=10000
    bAngularSpring=true
    bOverrideAngularDamping=true
    bUseGravityOverride=true
    gravityScale=0.35
    simulationSpace=RootRelative
    alphaBoolBlend {blendOption=Linear}
    boundBone {boneName=dyn_bang_r_1}
    componentAppliedLinearAccClamp=<3000 3000 3000>
    componentLinearAccScale=<0.4 0.4 0.4>
    componentLinearVelScale=<0.4 0.4 0.4>
    gravityOverride {z=-980}
    physicsBodyDefinitions [{
      boundBone {boneName=dyn_bang_r_1}
      boxExtents=<20 30 30>
      constraintSetup {angularLimitsMax=<0 0 7>, angularLimitsMin=<0 -15 -7>, angularTarget=<1 0 0>}
      localJointOffset=<4 0 0>
     }]
   }
   n13 = RigidBody!(ComponentPose=@n12, Alpha, OverrideWorldGravity:Vector="0,0,0") -> (Pose) {
    bOverrideWorldGravity=true
    overridePhysicsAsset=&/F_MED_AzureBlade_FaceAcc_Physics
    simulationSpace=BaseBoneSpace
    alphaBoolBlend {blendOption=Linear}
    alphaScaleBiasClamp {interpSpeedDecreasing=3, interpSpeedIncreasing=3}
    baseBoneRef {boneName=head}
    componentAppliedLinearAccClamp=<3000 3000 3000>
    simSpaceSettings {
     maxAngularAcceleration=1
     maxAngularVelocity=3
     maxLinearAcceleration=800
     maxLinearVelocity=400
     worldAlpha=1
     externalLinearDragV=<0.1 0.1 0.1>
    }
   }
   Save cached pose 'dynamicsCP'(Pose=Component To Local(ComponentPose=@n13)) {cacheName=dynamicsCP}
   n14 = "Constraint - dyn_braid_l_1 (P)"(ComponentPose=Local To Component(LocalPose=Use cached pose 'dynamicsCP'()), Alpha=1, ConstraintWeights_0=1) {
    alphaBoolBlend {blendOption=Linear}
    boneToModify {boneName=dyn_braid_l_1}
    constraintSetup [{targetBone {boneName=head}}]
    constraintWeights [1]
   }
   n15 = "Constraint - dyn_braid_l_2 (PPP)"!(ComponentPose=@n14, Alpha=1, ConstraintWeights_0=0.2, ConstraintWeights_1:float=0.3, ConstraintWeights_2:float=0.5) -> (Pose) {
    alphaBoolBlend {blendOption=Linear}
    boneToModify {boneName=dyn_braid_l_2}
    constraintSetup [{targetBone {boneName=head}}, {targetBone {boneName=neck_01}}, {targetBone {boneName=neck_02}}]
    constraintWeights [1,0,1]
   }
   n16 = "Constraint - dyn_braid_l_3 (PP)"!(ComponentPose=@n15, Alpha=1, ConstraintWeights_0=0.63, ConstraintWeights_1:float=0.37) -> (Pose) {
    alphaBoolBlend {blendOption=Linear}
    boneToModify {boneName=dyn_braid_l_3}
    constraintSetup [{targetBone {boneName=spine_05}}, {targetBone {boneName=clavicle_l}}]
    constraintWeights [1,1]
   }
   n17 = "Constraint - dyn_braid_l_4 (PPPP)"!(
    ComponentPose=@n16
    Alpha=1
    ConstraintWeights_0=0.61
    ConstraintWeights_1:float=0.24
    ConstraintWeights_2:float=0.11
    ConstraintWeights_3:float=0.04
   ) -> (Pose) {
    alphaBoolBlend {blendOption=Linear}
    boneToModify {boneName=dyn_braid_l_4}
    constraintSetup [
     {targetBone {boneName=spine_05}}
     {targetBone {boneName=clavicle_l}}
     {targetBone {boneName=spine_04}}
     {targetBone {boneName=upperarm_twist_01_l}}
    ]
    constraintWeights [1,1,1,1]
   }
   Save cached pose 'NoRBAN_SetupCP'(Pose=Blend Poses by bool(BlendPose_0=Use cached pose 'dynamicsCP'(), BlendPose_1=Component To Local(ComponentPose=@n17), bActiveValue=Get bRigidBodyEnabled()) {blendType=Linear}) {cacheName=NoRBAN_SetupCP}
  ]}
 EventGraph {nodes [
   n1 = Event BlueprintUpdateAnimation()
   > Set Delta Time X(Delta Time X=@n1.DeltaTimeX)
   > n2 = Sequence!(execute) -> (then_0, then_1, then_2:Exec)
   > n3 = Gravity Override State Function DEPRECATED(CharacterPartAnimInstance=Self-Reference(), GameTimeInSeconds=GetGameTimeInSeconds(), GravityOverrideParams=Get Gravity Override Params())
   > Set GravityOverride(GravityOverride=@n3)
   n4 = Get bIsCrouching()
   n5 = Get Delta Time X()
   Set isCrouching float(
    execute=@n2.then_1
    isCrouching float:double=FInterpTo(Current=Get isCrouching float(), Target="To Float (Boolean)"(InBool=@n4), DeltaTime=@n5, InterpSpeed=15)
   ) -> (Output_Get:double)
   n6 = Get bIsTargeting()
   > Set isTargeting float(isTargeting float:double=FInterpTo(Current=Get isTargeting float(), Target="To Float (Boolean)"(InBool=@n6), DeltaTime=@n5, InterpSpeed=15)) -> (Output_Get:double)
   n7 = Get HeadMesh() -> (HeadMesh:Skeletal Mesh Component Budgeted Object Reference)
   Is Valid(exec=@n2.then_2, InputObject=@n7)
   n8 = GetSocketBoneName!(self:Skinned Mesh Component Object Reference=@n7, InSocketName:Name=FACIAL_C_Jaw) -> (ReturnValue:Name=None)
   > n9 = Set isUsingFacialBones(isUsingFacialBones:bool="Not Equal Exactly (String)"(A="To String (Name)"(InName=@n8), B=None)) -> (Output_Get:bool)
   > Branch(Condition=@n9)
   n10 = Get HeadMesh()
   n11 = BreakTransform(InTransform=GetSocketTransform(self=@n10, InSocketName=FACIAL_R_CheekInner, TransformSpace=RTS_ParentBoneSpace))
   > Set FACIAL_R_CheekInner_Translation(FACIAL_R_CheekInner_Translation:Vector=@n11) -> (Output_Get:Vector)
   > Set FACIAL_R_CheekInner_Rotation(FACIAL_R_CheekInner_Rotation:Rotator=@n11.Rotation) -> (Output_Get:Rotator)
   n12 = BreakTransform(InTransform=GetSocketTransform(self=@n10, InSocketName=FACIAL_C_Nose, TransformSpace=RTS_ParentBoneSpace))
   > Set FACIAL_C_Nose_Translation(FACIAL_C_Nose_Translation:Vector=@n12) -> (Output_Get:Vector)
   > Set FACIAL_C_Nose_Rotation(FACIAL_C_Nose_Rotation:Rotator=@n12.Rotation) -> (Output_Get:Rotator)
   n13 = BreakTransform(InTransform=GetSocketTransform(self=@n10, InSocketName=FACIAL_C_NoseLower, TransformSpace=RTS_ParentBoneSpace))
   > Set FACIAL_C_NoseLower_Translation(FACIAL_C_NoseLower_Translation:Vector=@n13) -> (Output_Get:Vector)
   > Set FACIAL_C_NoseLower_Rotation(FACIAL_C_NoseLower_Rotation:Rotator=@n13.Rotation) -> (Output_Get:Rotator)
   n14 = BreakTransform(InTransform=GetSocketTransform(self=@n10, InSocketName=FACIAL_R_Nostril, TransformSpace=RTS_ParentBoneSpace))
   > Set FACIAL_R_Nostril_Translation(FACIAL_R_Nostril_Translation:Vector=@n14) -> (Output_Get:Vector)
   > Set FACIAL_R_Nostril_Rotation(FACIAL_R_Nostril_Rotation:Rotator=@n14.Rotation) -> (Output_Get:Rotator)
   n15 = BreakTransform(InTransform=GetSocketTransform(self=@n10, InSocketName=FACIAL_C_NoseTip, TransformSpace=RTS_ParentBoneSpace))
   > Set FACIAL_C_NoseTip_Translation(FACIAL_C_NoseTip_Translation:Vector=@n15) -> (Output_Get:Vector)
   > Set FACIAL_C_NoseTip_Rotation(FACIAL_C_NoseTip_Rotation:Rotator=@n15.Rotation) -> (Output_Get:Rotator)
   n16 = Event OnCharacterCustomizationLoaded()
   > Set HeadMesh(HeadMesh:Skeletal Mesh Component Budgeted Object Reference=Get Array Item with EFort Custom Part Type(Array=@n16.CustomMeshParts, PartType=Head)) -> (Output_Get:Skeletal Mesh Component Budgeted Object Reference)
   Event OnWeaponChanged()
   > Set isUsingPistol(isUsingPistol:double=0) -> (Output_Get:double)
   > n17 = Switch on EFortWeaponCoreAnimation!(execute, Selection:EFortWeaponCoreAnimation Enum=Get WeaponCoreAnim!() -> (WeaponCoreAnim)) -> (
    Melee:Exec
    Pistol:Exec
    Shotgun:Exec
    PaperBlueprint:Exec
    Rifle:Exec
    MeleeOneHand:Exec
    MachinePistol:Exec
    RocketLauncher:Exec
    GrenadeLauncher:Exec
    GoingCommando:Exec
    AssaultRifle:Exec
    TacticalShotgun:Exec
    SniperRifle:Exec
    TrapPlacement:Exec
    ShoulderLauncher:Exec
    AbilityDecoTool:Exec
    Crossbow:Exec
    C4:Exec
    RemoteControl:Exec
    DualWield:Exec
    AR_BullPup:Exec
    AR_ForwardGrip:Exec
    MedPackPaddles:Exec
    SMG_P90:Exec
    AR_DrumGun:Exec
    Consumable_Small:Exec
    Consumable_Large:Exec
    Balloon:Exec
    MountedTurret:Exec
    CreativeTool:Exec
    ExplosiveBow:Exec
    AshtonIndigo:Exec
    AshtonChicago:Exec
    MeleeDualWield:Exec
    Unarmed:Exec
   )
   Set isUsingPistol(execute=@n17.Pistol, isUsingPistol=1)
   Event BlueprintInitializeAnimation()
   > Set SkeletalMesh(SkeletalMesh:Skeletal Mesh Component Object Reference=GetOwnerSkeletalMeshComponent()) -> (Output_Get:Skeletal Mesh Component Object Reference)
  ]}
]
variables [
 Delta Time X {type=number}
 FACIAL_C_Nose_Rotation {type=Rotator}
 FACIAL_C_Nose_Translation {type=Vector}
 FACIAL_C_NoseLower_Rotation {type=Rotator}
 FACIAL_C_NoseLower_Translation {type=Vector}
 FACIAL_C_NoseTip_Rotation {type=Rotator}
 FACIAL_C_NoseTip_Translation {type=Vector}
 FACIAL_R_CheekInner_Rotation {type=Rotator}
 FACIAL_R_CheekInner_Translation {type=Vector}
 FACIAL_R_Nostril_Rotation {type=Rotator}
 FACIAL_R_Nostril_Translation {type=Vector}
 Gravity Override Params {type=GravityOverrideParamsStruct}
 GravityOverride {type=Vector}
 HeadMesh {type=SkeletalMeshComponentBudgeted}
 isCrouching float {type=number}
 isTargeting float {type=number}
 isUsingFacialBones {type=boolean}
 isUsingPistol {type=number}
 SkeletalMesh {type=SkeletalMeshComponent}
]
